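# Animation Blueprint for the left Attic Drill pickaxe cosmetic that adds secondary motion to its three vines.
# The input pose gets fixed additive offsets on the claw bones, then spring-based AnimDynamics on each vine's first bone.
# Additive bone rotations then sway the first and second vine segments.
# The Event Graph drives these from real-time pulsating values, phase-shifting the second segments so the idle wiggle trails along each vine.
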AnimBlueprint /BRCosmetics/Weapons/FORT_Melee/Pickaxe_AtticDrill/Meshes/AtticDrill_Axe_L_AnimBP
target_skeleton=&/AtticDrill_Axe_Skeleton
graphs [
 AnimGraph {nodes [
   AnimGraphNode_SubInput_0 = Input Pose()
   n1 = "Transform (Modify) Bone - Bone: Claws_main_l"(ComponentPose=Local To Component(LocalPose=@AnimGraphNode_SubInput_0), Alpha=1, Translation="0,0,0", Rotation="180,0,0", Scale="1,1,1") {
    rotationMode=BMM_Additive
    rotationSpace=BCS_BoneSpace
    translationMode=BMM_Additive
    boneToModify {boneName=Claws_main_l}
    rotation {pitch=180}
   }
   n2 = "Transform (Modify) Bone - Bone: Claws_l"(ComponentPose=@n1, Alpha=1, Translation="-0.022,3,20", Rotation="5,5,-108.859", Scale="1,1,1") {
    rotationMode=BMM_Additive
    rotationSpace=BCS_BoneSpace
    translationMode=BMM_Additive
    boneToModify {boneName=Claws_l}
    rotation=<rot 5 5 -108.859>
    translation=<-0.022 3 20>
   }
   n3 = Anim Dynamics - Bone: vine_a_01_l(ComponentPose=@n2, Alpha=1, ExternalForce="0,0,0") {
    angularDampingOverride=1
    angularSpringConstant=4000
    bAngularSpring=true
    bOverrideAngularDamping=true
    boundBone {boneName=vine_a_01_l}
    componentAppliedLinearAccClamp=<3000 3000 2500>
    componentLinearAccScale=<0.3 0.3 0.3>
    componentLinearVelScale=<0.3 0.3 0.3>
    physicsBodyDefinitions [{
      boundBone {boneName=vine_a_01_l}
      boxExtents=<21 21 21>
      constraintSetup {angularLimitsMax=<0 5 5>, angularLimitsMin=<0 -5 -5>, angularTarget=<1 0 0>}
      localJointOffset=<0.1 0 0>
     }]
    simSpaceSettings {maxAngularAcceleration=35, maxAngularVelocity=7, simSpaceAngularAlpha=1}
   }
   n4 = Anim Dynamics - Bone: vine_b_01_l(ComponentPose=@n3, Alpha=1, ExternalForce="0,0,0") {
    angularDampingOverride=1
    angularSpringConstant=6000
    bAngularSpring=true
    bOverrideAngularDamping=true
    boundBone {boneName=vine_b_01_l}
    componentAppliedLinearAccClamp=<3000 3000 2500>
    componentLinearAccScale=<0.3 0.3 0.3>
    componentLinearVelScale=<0.3 0.3 0.3>
    physicsBodyDefinitions [{
      boundBone {boneName=vine_b_01_l}
      boxExtents=<27 27 27>
      constraintSetup {angularLimitsMax=<0 8 0>, angularLimitsMin=<0 0 -8>, angularTarget=<1 0 0>}
      localJointOffset=<0.1 0 0>
     }]
    simSpaceSettings {maxAngularAcceleration=35, maxAngularVelocity=7, simSpaceAngularAlpha=1}
   }
   n5 = Anim Dynamics - Bone: vine_c_01_l(ComponentPose=@n4, Alpha=1, ExternalForce="0,0,0") {
    angularDampingOverride=1
    angularSpringConstant=7000
    bAngularSpring=true
    bOverrideAngularDamping=true
    boundBone {boneName=vine_c_01_l}
    componentAppliedLinearAccClamp=<3000 3000 2500>
    componentLinearAccScale=<0.3 0.3 0.3>
    componentLinearVelScale=<0.3 0.3 0.3>
    physicsBodyDefinitions [{
      boundBone {boneName=vine_c_01_l}
      boxExtents=<29 29 29>
      constraintSetup {angularLimitsMax=<0 0 9>, angularLimitsMin=<0 -9 0>, angularTarget=<1 0 0>}
      localJointOffset=<0.1 0 0>
     }]
    simSpaceSettings {maxAngularAcceleration=35, maxAngularVelocity=7, simSpaceAngularAlpha=1}
   }
   n6 = Get Vine 01 Z() -> (Vine 01 Z:Rotator)
   n7 = "Transform (Modify) Bone - Bone: vine_b_01_l"(
    ComponentPose="Transform (Modify) Bone - Bone: vine_a_01_l"(ComponentPose=@n5, Alpha=1, Translation="0,0,0", Rotation="Get Vine 01 Y "() -> ("Vine 01 Y ":Rotator), Scale="1,1,1") {rotationMode=BMM_Additive, rotationSpace=BCS_BoneSpace, boneToModify {boneName=vine_a_01_l}}
    Alpha=1
    Translation="0,0,0"
    Rotation=@n6
    Scale="1,1,1"
   ) {rotationMode=BMM_Additive, rotationSpace=BCS_BoneSpace, boneToModify {boneName=vine_b_01_l}}
   n8 = Get Vine 02 YZ() -> (Vine 02 YZ:Rotator)
   n9 = "Transform (Modify) Bone - Bone: vine_b_02_l"(
    ComponentPose="Transform (Modify) Bone - Bone: vine_a_02_l"(
     ComponentPose="Transform (Modify) Bone - Bone: vine_c_01_l"(ComponentPose=@n7, Alpha=1, Translation="0,0,0", Rotation=@n6, Scale="1,1,1") {rotationMode=BMM_Additive, rotationSpace=BCS_BoneSpace, boneToModify {boneName=vine_c_01_l}}
     Alpha=1
     Translation="0,0,0"
     Rotation=@n8
     Scale="1,1,1"
    ) {rotationMode=BMM_Additive, rotationSpace=BCS_BoneSpace, boneToModify {boneName=vine_a_02_l}}
    Alpha=1
    Translation="0,0,0"
    Rotation=@n8
    Scale="1,1,1"
   ) {rotationMode=BMM_Additive, rotationSpace=BCS_BoneSpace, boneToModify {boneName=vine_b_02_l}}
   n10 = "Get Vine 02 "() -> ("Vine 02 ":Rotator)
   Output Pose(Result=Component To Local(ComponentPose="Transform (Modify) Bone - Bone: vine_c_02_l"(ComponentPose=@n9, Alpha=1, Translation="0,0,0", Rotation=@n10, Scale="1,1,1") {rotationMode=BMM_Additive, rotationSpace=BCS_BoneSpace, boneToModify {boneName=vine_c_02_l}})) {layerGroup=None}
  ]}
 EventGraph {nodes [
   Event BlueprintUpdateAnimation()
   > Sequence!(execute) -> (then_0, then_1, then_2:Exec)
   n1 = GetRealTimeSeconds()
   n2 = float - float(A=MakePulsatingValue!(InCurrentTime:float=@n1, InPulsesPerSecond:float=0.3, InPhase:float=0) -> (ReturnValue:float), B=0.5)
   > Set Vine 01 Z(Vine 01 Z:Rotator=MakeRotator(Yaw=float * float(A=@n2, B=6) {is=title K2Node_PromotableOperator})) -> (Output_Get:Rotator)
   n3 = float - float(A=MakePulsatingValue(InCurrentTime=@n1, InPulsesPerSecond=0.3, InPhase=0), B=0.5)
   > "Set Vine 01 Y "("Vine 01 Y ":Rotator=MakeRotator(Pitch=float * float(A=@n3, B=6))) -> (Output_Get:Rotator)
   n4 = float * float(A=float - float(A=MakePulsatingValue(InCurrentTime=@n1, InPulsesPerSecond=0.3, InPhase=0.3), B=0.5), B=5)
   > Set Vine 02 YZ(Vine 02 YZ:Rotator=MakeRotator(Pitch=@n4, Yaw=@n4)) -> (Output_Get:Rotator)
   n5 = float - float(A=MakePulsatingValue(InCurrentTime=@n1, InPulsesPerSecond=0.3, InPhase=0.3), B=0.5)
   > "Set Vine 02 "("Vine 02 ":Rotator=MakeRotator(Pitch=float * float(A=@n5, B=6))) -> (Output_Get:Rotator)
   TryGetPawnOwner()
  ]}
]
variables [
 "Vine 01 Y " {type=Rotator}
 Vine 01 Z {type=Rotator}
 "Vine 02 " {type=Rotator}
 Vine 02 YZ {type=Rotator}
]
